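# Physics asset for the cloth on the GreatPoolSwim medium male body, covering a short dynamic cloak bone chain and the lower body around it.
# Capsules on the pelvis and both thighs give body volume, with thigh joints left free in swing and twist.
# The cloak base locks angularly to the pelvis, and two further cloak capsules chain from it with limited swing, wider at the tip, and very tight twist.
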PhysicsAsset /BRCosmetics/Characters/Player/Male/Medium/Bodies/M_MED_GreatPoolSwim/Meshes/M_MED_GreatPoolSwim_Cloth_Physics1
bodies {
 dyn_cloak_01 [{
   shapeType=Capsule
   length=3.8457
   radius=2.6915
   shapeName=dyn_cloak_01_capsule
   center=<5.2879 -0.2113 1.1436>
   rotation=<rot 6.0767 90.2961 -85.3966>
  }]
 dyn_cloak_02 [{
   shapeType=Capsule
   length=3.9181
   radius=2.6249
   shapeName=dyn_cloak_02_capsule
   center=<4.7988 -1.0721 -0.2204>
   rotation=<rot -19.949 -88.9225 -90.0899>
  }]
 dyn_cloak_base [{shapeType=Capsule, length=0.505, radius=0.505, shapeName=dyn_cloak_base_capsule, rotation=<rot 0 -90 0>}]
 pelvis [{shapeType=Capsule, length=8.5938, radius=10.2812, center=<-1.4216 -1.0559 -0.0001>}]
 thigh_l [{shapeType=Capsule, length=28.9688, radius=9.5938, center=<-15.1488 -0.9527 2.4766>, rotation=<rot 88.5265 -9.6052 -10.4338>}]
 thigh_r [{shapeType=Capsule, length=28.2188, radius=9, center=<17.057 0.2342 -2.0844>, rotation=<rot 88.5265 -9.6056 -12.5396>}]
}
constraints [
 {
  bone1Name=thigh_l
  bone2Name=pelvis
  swing1LimitDegrees=45
  swing1Motion=Free
  swing2LimitDegrees=45
  swing2Motion=Free
  twistLimitDegrees=45
  twistMotion=Free
 }
 {
  bone1Name=thigh_r
  bone2Name=pelvis
  swing1LimitDegrees=45
  swing1Motion=Free
  swing2LimitDegrees=45
  swing2Motion=Free
  twistLimitDegrees=45
  twistMotion=Free
 }
 {bone1Name=dyn_cloak_02, bone2Name=dyn_cloak_01, swing1LimitDegrees=30, swing2LimitDegrees=30, twistLimitDegrees=2}
 {bone1Name=dyn_cloak_01, bone2Name=dyn_cloak_base, swing1LimitDegrees=20, swing2LimitDegrees=20, twistLimitDegrees=2}
 {
  bone1Name=dyn_cloak_base
  bone2Name=pelvis
  swing1LimitDegrees=45
  swing1Motion=Locked
  swing2LimitDegrees=45
  swing2Motion=Locked
  twistLimitDegrees=45
  twistMotion=Locked
 }
]
