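# Anim Blueprint for the male medium GreatPoolSwim cosmetic body, layering secondary motion over the pose copied from the attached parent mesh.
# The copied pose runs through the body's Control Rig, then a RigidBody cloth simulation in pelvis bone space with scaled-down component motion.
# An additive rotation on the dynamic holster joint blends by an alpha computed each update from the right thigh's Euler Y delta relative to the pelvis.
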AnimBlueprint /BRCosmetics/Characters/Player/Male/Medium/Bodies/M_MED_GreatPoolSwim/Meshes/M_MED_GreatPoolSwim_AnimBP
target_skeleton=&/M_MED_GreatPoolSwim_Skeleton
graphs [
 AnimGraph {nodes [
   n1 = Control Rig(Source=Copy Pose From Mesh() {bUseAttachedParent=true}) {
    alphaBoolBlend {blendOption=Linear}
    controlRigAssetReference {blueprintRigClass=&/M_MED_GreatPoolSwim_CtrlRig.M_MED_GreatPoolSwim_CtrlRig_C}
   }
   n2 = RigidBody!(ComponentPose=Local To Component(LocalPose=@n1), Alpha, OverrideWorldGravity:Vector="0,0,-980", ExternalForce) -> (Pose) {
    bOverrideWorldGravity=true
    overridePhysicsAsset=&/M_MED_GreatPoolSwim_Cloth_Physics1
    simulationSpace=BaseBoneSpace
    alphaBoolBlend {blendOption=Linear}
    baseBoneRef {boneName=pelvis}
    componentAppliedLinearAccClamp=<1000 1000 1000>
    componentLinearAccScale=<0.3 0.3 0.3>
    componentLinearVelScale=<0.3 0.3 0.3>
   }
   n3 = "Transform (Modify) Bone - Bone: dyn_holster_joint"(ComponentPose=@n2, Alpha=Get holster delta() -> (holster delta:float), Translation="0,0,0", Rotation="20,0,-10", Scale="1,1,1") {
    rotationMode=BMM_Additive
    rotationSpace=BCS_BoneSpace
    translationMode=BMM_Additive
    translationSpace=BCS_BoneSpace
    alphaBoolBlend {blendOption=Linear}
    boneToModify {boneName=dyn_holster_joint}
    rotation {pitch=20, roll=-10}
   }
   Output Pose(Result=Component To Local(ComponentPose=@n3))
  ]}
 EventGraph {nodes [
   Event BlueprintUpdateAnimation()
   > Set holster delta(holster delta:float=GetAlphaFromReferenceTransformDelta(
     SkeletalMeshComponent=GetOwnerSkeletalMeshComponent()
     BoneName=thigh_r
     BaseName=pelvis
     DeltaComponentType=EulerY
     InputRange=(X=6,Y=26)
     TwistAxis="1,0,0"
    )) -> (Output_Get:float)
   TryGetPawnOwner()
  ]}
]
variables [holster delta {type=number}]
